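# Niagara system for the DarkFlame vehicle boost cosmetic: a dark flame exhaust with violet-blue highlights and a trail.
# A ribbon emitter draws three trail layers, switching offset, velocity and fade curves by game mode.
# A sprite emitter merges four spawn groups of clouds, embers and darts, while a burst emitter builds the sprite-and-mesh core.
# User colors replace defaults when bIsPainted is set, and ColorVariantID adjusts the ribbon material.
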NiagaraSystem /VehicleCosmetics/Boosts/Boost_DarkFlame/FX/NS_Boost_DarkFlame
dependencies {
 usedDataInterfaces [
  &NiagaraDataInterfaceCurve
  &NiagaraDataInterfaceSpriteRendererInfo
  &NiagaraDataInterfaceSkeletalMesh
 ]
 usedDynamicInputs []
 usedModules [
  &self:SystemSpawnScript.SystemScriptSource.SystemScriptGraph.NiagaraNodeAssignment_0.SetVariables_#1
  &SystemState
 ]
}
emitters {
 NE_VC_SingleBoost_Ribbon_Parent {
  emitterUpdateScript [
   /SetVariables(Emitter.FlatColor=HlslExpression() {hlslExpression="(User.ColorVariantID) == 7 || (User.ColorVariantID) == 14 ? true : false"}) {isSetParameters=true}
   &EmitterState()
   &VC_FilteredSocketCount() {enabled=false}
   &VC_RateQualityHandler(High Spawn Rate=60, Low Spawn Rate=60, Medium Spawn Rate=60, Use Custom Rate?=true)
   &VC_SpawnRibbon_v1(SpawnRate=$Emitter.Spawn Rate)
   /SetVariables(Emitter.FlatColorCheck:NiagaraBool=&Boolean_NotOperation(Bool=$Emitter.FlatColor)) {isSetParameters=true}
  ]
  particleSpawnScript [
   &VC_PaintedVariantsOverride() {enabled=false}
   &VC_InitializeRibbon_v1(
    Color=&VC_DynamicScaleColor(
     Color to Scale:LinearColor=&MakeCustomLinearColorFromBool(Bool=$User.bIsPainted, True Color=$User.MTrailColor, False Color=<rgba 0.6722 0 5 5>)
     RGB Scale:Vector3f=<1 1 1>
     Alpha Scale:NiagaraFloat=1
    )
    Custom Offset=true
    Local Vector Offset=&AddVectorToPosition(
     Position=/VehicleCosmetics/FX/Niagara/Modules/GameMode/VC_SetVectorByGameModeID(
      SameValueForLockerAndBR?:NiagaraBool=false
      Locker Vector:Vector3f=<100 0 0>
      Battle Royale Vector:Vector3f=<200 0 0>
      Rocket Racing Vector:Vector3f=<-120 0 0>
      GameModeID:NiagaraInt32=$System.GameModeID
     )
     Vector=$System.FinalPosition
    )
    Offset Amount=1
    Ribbon Width=75
   )
   &AddVelocity(
    Velocity=&MakeVector(
     X=/VehicleCosmetics/FX/Niagara/Modules/GameMode/VC_SetFloatByGameModeID(
      SameValueForLockerAndBR?:NiagaraBool=false
      Locker Float:NiagaraFloat=-1200
      Battle Royale Float:NiagaraFloat=-1000
      Rocket Racing Float:NiagaraFloat=-900
      GameModeID:NiagaraInt32=$System.GameModeID
     )
     Y=0
     Z=0
    )
    Velocity Speed Scale=3
   ) {module=AddVelocity001}
   /SetVariables(Particles.SdRibbonWidth=100) {isSetParameters=true}
  ]
  particleUpdateScript [
   /SetVariables(
    Particles.AlternateColor:LinearColor=&VC_DynamicScaleColor(
     Color to Scale:LinearColor=&MakeCustomLinearColorFromBool(Bool=$User.bIsPainted, True Color=$User.GreyTColor, False Color=<rgba 1 1 1 1>)
     RGB Scale:Vector3f=<1 1 1>
     Alpha Scale:NiagaraFloat=&VC_SetFloatByGameModeID(
      SameValueForLockerAndBR?:NiagaraBool=false
      Locker Float:NiagaraFloat=&FloatFromCurve(FloatCurve=DataInterface() {propertyValues {Curve {keys [
           {arriveTangent=0.0586, time=0, value=0}
           {arriveTangent=2.5, interpMode=RCIM_Cubic, leaveTangent=2.5, tangentMode=RCTM_Break, time=0.11, value=0}
           {arriveTangent=0.1318, interpMode=RCIM_Cubic, leaveTangent=0.1318, tangentMode=RCTM_User, time=0.35, value=1}
           <at 0.7 0 RCIM_Cubic>
          ]}}})
      Battle Royale Float:NiagaraFloat=&FloatFromCurve(FloatCurve=DataInterface() {propertyValues {Curve {keys [
           {arriveTangent=0.0586, time=0, value=0}
           {arriveTangent=2.5, interpMode=RCIM_Cubic, leaveTangent=2.5, tangentMode=RCTM_Break, time=0.2, value=0}
           {arriveTangent=0.1318, interpMode=RCIM_Cubic, leaveTangent=0.1318, tangentMode=RCTM_User, time=0.35, value=1}
           <at 0.7 0 RCIM_Cubic>
          ]}}})
      Rocket Racing Float:NiagaraFloat=1
      GameModeID:NiagaraInt32=$System.GameModeID
     )
    )
    Particles.SndColor:LinearColor=&VC_DynamicScaleColor(
     Color to Scale:LinearColor=&MakeCustomLinearColorFromBool(Bool=$User.bIsPainted, True Color=$User.SndTColor, False Color=<rgba 0.0084 0.0099 0.0938 1>)
     RGB Scale:Vector3f=<10 10 10>
     Alpha Scale:NiagaraFloat=&VC_SetFloatByGameModeID(
      SameValueForLockerAndBR?:NiagaraBool=false
      Locker Float:NiagaraFloat=&FloatFromCurve(FloatCurve=DataInterface() {propertyValues {Curve {keys [
           {arriveTangent=0.0586, time=0, value=0}
           {arriveTangent=2.5, interpMode=RCIM_Cubic, leaveTangent=2.5, tangentMode=RCTM_Break, time=0.11, value=0}
           {arriveTangent=0.1318, interpMode=RCIM_Cubic, leaveTangent=0.1318, tangentMode=RCTM_User, time=0.35, value=1}
           <at 0.7 0 RCIM_Cubic>
          ]}}})
      Battle Royale Float:NiagaraFloat=&FloatFromCurve(FloatCurve=DataInterface() {propertyValues {Curve {keys [
           {arriveTangent=0.0586, time=0, value=0}
           {arriveTangent=2.5, interpMode=RCIM_Cubic, leaveTangent=2.5, tangentMode=RCTM_Break, time=0.2, value=0}
           {arriveTangent=0.1318, interpMode=RCIM_Cubic, leaveTangent=0.1318, tangentMode=RCTM_User, time=0.35, value=1}
           <at 0.7 0 RCIM_Cubic>
          ]}}})
      Rocket Racing Float:NiagaraFloat=&FloatFromCurve(FloatCurve=DataInterface() {propertyValues {Curve {keys [
           {arriveTangent=0.0586, time=0, value=0}
           {arriveTangent=2.5, interpMode=RCIM_Cubic, leaveTangent=2.5, tangentMode=RCTM_Break, time=0.11, value=0}
           {arriveTangent=0.1318, interpMode=RCIM_Cubic, leaveTangent=0.1318, tangentMode=RCTM_User, time=0.35, value=1}
           <at 0.7 0 RCIM_Cubic>
          ]}}})
      GameModeID:NiagaraInt32=$System.GameModeID
     )
    )
   ) {isSetParameters=true}
   &ParticleState()
   &ScaleColor(
    Scale Alpha=&VC_SetFloatByGameModeID(
     SameValueForLockerAndBR?:NiagaraBool=false
     Locker Float:NiagaraFloat=&FloatFromCurve(FloatCurve=DataInterface() {propertyValues {Curve {keys [
          {arriveTangent=0.0586, time=0, value=0}
          {arriveTangent=2.5, interpMode=RCIM_Cubic, leaveTangent=2.5, tangentMode=RCTM_Break, time=0.11, value=0}
          {arriveTangent=0.1318, interpMode=RCIM_Cubic, leaveTangent=0.1318, tangentMode=RCTM_User, time=0.35, value=1}
          <at 0.7 0 RCIM_Cubic>
         ]}}})
     Battle Royale Float:NiagaraFloat=&FloatFromCurve(FloatCurve=DataInterface() {propertyValues {Curve {keys [
          {arriveTangent=0.0586, time=0, value=0}
          {arriveTangent=2.5, interpMode=RCIM_Cubic, leaveTangent=2.5, tangentMode=RCTM_Break, time=0.25, value=0}
          {arriveTangent=0.1318, interpMode=RCIM_Cubic, leaveTangent=0.1318, tangentMode=RCTM_User, time=0.35, value=1}
          <at 0.7 0 RCIM_Cubic>
         ]}}})
     Rocket Racing Float:NiagaraFloat=&FloatFromCurve(FloatCurve=DataInterface() {propertyValues {Curve {keys [
          {arriveTangent=0.0586, interpMode=RCIM_Cubic, tangentMode=RCTM_Break, time=0, value=0}
          {arriveTangent=0.1318, interpMode=RCIM_Cubic, leaveTangent=0.1318, tangentMode=RCTM_Break, time=0.1, value=1}
          <at 0.7 0 RCIM_Cubic>
         ]}}})
     GameModeID:NiagaraInt32=$System.GameModeID
    )
    ScaleA=true
    ScaleRGB=true
   )
   &InheritSourceMovement(
    Applied Position Scale=&Multiply_VectorByFloat(
     Vector=<1 1 1>
     Float=&FloatFromCurve(FloatCurve=DataInterface() {propertyValues {Curve {keys [{interpMode=RCIM_Cubic, tangentMode=RCTM_Break, time=0.5, value=1}, {interpMode=RCIM_Cubic, tangentMode=RCTM_Break, time=1, value=0}]}}})
    )
    ApplyForce=false
    ApplyPosition=true
    ApplyPositionFalloff=false
   )
   &SolveForcesAndVelocity(Clamp Velocity=false, Limit Acceleration=false)
   &VC_UpdateRibbon_v1()
   &DynamicMaterialParameters(
    Index 0 Param 1=HlslExpression() {hlslExpression="(User.ColorVariantID) == 7 ? 0.2 : User.ColorVariantID == 14 ? 0.4 : 0 "}
    Module.Index0Float0Write=true
    Module.Index0Float1Write=true
    Module.Index0Float2Write=true
    Module.Index0Float3Write=true
   )
  ]
  rendererClasses [&NiagaraRibbonRendererProperties]
  renderers [
   {rendererClass=&NiagaraRibbonRendererProperties}
   {rendererClass=&NiagaraRibbonRendererProperties, rendererIndex=1}
   {rendererClass=&NiagaraRibbonRendererProperties, rendererIndex=2}
  ]
 }
 NE_VC_SingleBoost_Sprite_Parent {
  emitterUpdateScript [
   &EmitterState()
   &SpawnRate(SpawnRate=50, Use Spawn Probability=false)
   &SpawnRate(SpawnGroup=1, SpawnRate=40, Use Spawn Probability=false) {module=SpawnRate001}
   &SpawnRate(SpawnGroup=2, SpawnRate=&RandomRangeFloat(Minimum=125, Maximum=175), Use Spawn Probability=false) {module=SpawnRate002}
   &SpawnRate(SpawnGroup=3, SpawnRate=10, Use Spawn Probability=false) {module=SpawnRate003}
  ]
  particleSpawnScript [
   &VC_PaintedVariantsOverride() {enabled=false}
   &InitializeParticle(
    Color=&VC_DynamicScaleColor(
     Color to Scale:LinearColor=&MakeCustomLinearColorFromBool(Bool=$User.bIsPainted, True Color=$User.CloudsColor, False Color=<rgba 0 0 0 0.45>)
     RGB Scale:Vector3f=<1 1 1>
     Alpha Scale:NiagaraFloat=1
    )
    Lifetime=/VehicleCosmetics/FX/Niagara/Modules/MergeEmitters/VC_FloatBySpawnGroup(
     SpawnGroup or ExecIndex:NiagaraInt32=0
     CustomInt:NiagaraInt32=0
     MaximumSpawnGroups:NiagaraInt32=4
     Float SG 0:NiagaraFloat=&RandomRangeFloat(Minimum=0.15, Maximum=0.3)
     Float SG 1:NiagaraFloat=&RandomRangeFloat(Minimum=0.25, Maximum=0.35)
     Float SG 2:NiagaraFloat=&RandomRangeFloat(Minimum=0.25, Maximum=0.45)
     Float SG 3:NiagaraFloat=&RandomRangeFloat(Minimum=0.25, Maximum=0.5)
     Float SG 4:NiagaraFloat=0
     Float SG 5:NiagaraFloat=0
    )
    Position Offset=&AddVectorToPosition(Position=<-60 0 0>, Vector=$System.FinalPosition)
    Position Offset Coordinate Space=::Local
    Sprite Rotation Mode=::NewEnumerator2
    Sprite Size Mode=::NewEnumerator3
    Uniform Sprite Size=&VC_FloatBySpawnGroup(
     SpawnGroup or ExecIndex:NiagaraInt32=0
     CustomInt:NiagaraInt32=0
     MaximumSpawnGroups:NiagaraInt32=4
     Float SG 0:NiagaraFloat=&RandomRangeFloat(
      Minimum=&Multiply_Float(
       A=20
       B=&FloatFromCurve(FloatCurve=DataInterface() {propertyValues {Curve {keys [<at 0.8 5>,<at 0.8 5>]}}})
      )
      Maximum=&Multiply_Float(
       A=25
       B=&FloatFromCurve(FloatCurve=DataInterface() {propertyValues {Curve {keys [<at 0.8 5>,<at 0.8 5>]}}})
      )
     )
     Float SG 1:NiagaraFloat=&RandomRangeFloat(
      Minimum=&Multiply_Float(
       A=17
       B=&FloatFromCurve(FloatCurve=DataInterface() {propertyValues {Curve {keys [<at 0.8 5>,<at 0.8 5>]}}})
      )
      Maximum=&Multiply_Float(
       A=20
       B=&FloatFromCurve(FloatCurve=DataInterface() {propertyValues {Curve {keys [<at 0.8 5>,<at 0.8 5>]}}})
      )
     )
     Float SG 2:NiagaraFloat=&RandomRangeFloat(Minimum=1, Maximum=9)
     Float SG 3:NiagaraFloat=1
     Float SG 4:NiagaraFloat=0
     Float SG 5:NiagaraFloat=0
    )
    UsePositionOffset=true
    Write Material Random=true
   )
   &AddVelocity(Velocity=/VehicleCosmetics/FX/Niagara/Modules/MergeEmitters/VC_VectorBySpawnGroup(
     SpawnGroup or ExecIndex:NiagaraInt32=0
     CustomInt:NiagaraInt32=0
     MaximumSpawnGroups:NiagaraInt32=4
     Vector Group 0:Vector3f=<-450 0 0>
     Vector Group 1:Vector3f=<0 0 0>
     Vector Group 2:Vector3f=<-300 0 0>
     Vector Group 3:Vector3f=<-1000 0 0>
     Vector Group 4:Vector3f=<0 0 0>
     Vector Group 5:Vector3f=<0 0 0>
    ))
   &ShapeLocation(
    MaskSpawnGroup=false
    Override Seed=false
    Sphere Radius=&VC_FloatBySpawnGroup(
     SpawnGroup or ExecIndex:NiagaraInt32=0
     CustomInt:NiagaraInt32=0
     MaximumSpawnGroups:NiagaraInt32=4
     Float SG 0:NiagaraFloat=2
     Float SG 1:NiagaraFloat=2
     Float SG 2:NiagaraFloat=&RandomRangeFloat(Minimum=10, Maximum=45)
     Float SG 3:NiagaraFloat=25
     Float SG 4:NiagaraFloat=0
     Float SG 5:NiagaraFloat=0
    )
    UseOwnerScale=true
   )
   /SetVariables(
    Particles.DarkEmbSize:Vector2f=&RandomRangeVector2D(Minimum=<1.5 15>, Maximum=<2.5 25>)
    Particles.DartSpriteSize:Vector2f=&RandomRangeVector2D(Minimum=<4 60>, Maximum=<5 50>)
   ) {isSetParameters=true}
  ]
  particleUpdateScript [
   /SetVariables(
    Particles.CTagVis:NiagaraInt32=&SpawnGroup()
    Particles.CloudColor:LinearColor=&MakeCustomLinearColorFromBool(Bool=$User.bIsPainted, True Color=$User.CloudsColor, False Color=<rgba 0 0 0 0.75>)
    Particles.DEmbColor:LinearColor=&VC_DynamicScaleColor(
     Color to Scale:LinearColor=&MakeCustomLinearColorFromBool(Bool=$User.bIsPainted, True Color=$User.DartsColor, False Color=<rgba 1.5 0.75 5 1>)
     RGB Scale:Vector3f=<1 1 1>
     Alpha Scale:NiagaraFloat=1
    )
    Particles.DartsColor:LinearColor=&VC_DynamicScaleColor(
     Color to Scale:LinearColor=&MakeCustomLinearColorFromBool(Bool=$User.bIsPainted, True Color=$User.DartsColor, False Color=<rgba 0.1 0.03 1 1>)
     RGB Scale:Vector3f=<5 5 5>
     Alpha Scale:NiagaraFloat=1
    )
   ) {isSetParameters=true}
   &ParticleState()
   &ScaleColor(
    Scale Alpha=&FloatFromCurve(FloatCurve=DataInterface() {propertyValues {Curve {
        defaultValue=1
        keys [
         {interpMode=RCIM_Cubic, tangentMode=RCTM_User, time=0, value=0}
         {interpMode=RCIM_Cubic, tangentMode=RCTM_User, time=0.5, value=1}
         {tangentMode=RCTM_User, time=1, value=0}
        ]
       }}})
    ScaleA=true
    ScaleRGB=true
   )
   &SubUVAnimation(
    Output Alternate Renderer Binding=true
    "Sprite Renderer "=DataInterface() {propertyValues {SpriteRenderer=&self:NE_VC_SingleBoost_Sprite_Parent_0.NiagaraSpriteRendererProperties_0}}
    UseEndFrame=false
    UseStartFrame=false
   )
   &SubUVAnimation(
    Output Alternate Renderer Binding=true
    "Sprite Renderer "=DataInterface() {propertyValues {SpriteRenderer=&self:NE_VC_SingleBoost_Sprite_Parent_0.NiagaraSpriteRendererProperties_1}}
    UseEndFrame=false
    UseStartFrame=false
   ) {module=SubUVAnimation001}
   &SubUVAnimation(
    Output Alternate Renderer Binding=true
    "Sprite Renderer "=DataInterface() {propertyValues {SpriteRenderer=&self:NE_VC_SingleBoost_Sprite_Parent_0.NiagaraSpriteRendererProperties_2}}
    UseEndFrame=false
    UseStartFrame=false
   ) {module=SubUVAnimation002}
   &CurlNoiseForce(
    Bias Noise=false
    Noise Frequency=&VC_FloatBySpawnGroup(
     SpawnGroup or ExecIndex:NiagaraInt32=0
     CustomInt:NiagaraInt32=0
     MaximumSpawnGroups:NiagaraInt32=4
     Float SG 0:NiagaraFloat=0
     Float SG 1:NiagaraFloat=0
     Float SG 2:NiagaraFloat=35
     Float SG 3:NiagaraFloat=0
     Float SG 4:NiagaraFloat=0
     Float SG 5:NiagaraFloat=0
    )
    Noise Strength=&VC_FloatBySpawnGroup(
     SpawnGroup or ExecIndex:NiagaraInt32=0
     CustomInt:NiagaraInt32=0
     MaximumSpawnGroups:NiagaraInt32=4
     Float SG 0:NiagaraFloat=0
     Float SG 1:NiagaraFloat=0
     Float SG 2:NiagaraFloat=750
     Float SG 3:NiagaraFloat=0
     Float SG 4:NiagaraFloat=0
     Float SG 5:NiagaraFloat=0
    )
    Pan Noise=false
   )
   &Drag(
    Drag=&VC_FloatBySpawnGroup(
     SpawnGroup or ExecIndex:NiagaraInt32=0
     CustomInt:NiagaraInt32=0
     MaximumSpawnGroups:NiagaraInt32=4
     Float SG 0:NiagaraFloat=0
     Float SG 1:NiagaraFloat=0
     Float SG 2:NiagaraFloat=0
     Float SG 3:NiagaraFloat=&FloatFromCurve(FloatCurve=DataInterface() {propertyValues {Curve {defaultValue=1, keys [{interpMode=RCIM_Cubic, tangentMode=RCTM_User, time=0, value=1}, {tangentMode=RCTM_User, time=1, value=0}]}}})
     Float SG 4:NiagaraFloat=0
     Float SG 5:NiagaraFloat=0
    )
    Use Linear Drag=true
    Use Rotational Drag=false
   )
   &SolveForcesAndVelocity(Clamp Velocity=false, Limit Acceleration=false)
  ]
  rendererClasses [&NiagaraSpriteRendererProperties]
  renderers [
   {rendererClass=&NiagaraSpriteRendererProperties}
   {rendererClass=&NiagaraSpriteRendererProperties, rendererIndex=1}
   {rendererClass=&NiagaraSpriteRendererProperties, rendererIndex=2}
   {rendererClass=&NiagaraSpriteRendererProperties, rendererIndex=3}
  ]
 }
 NE_VC_SingleBoost_Sprite_Parent001 {
  emitterUpdateScript [
   &EmitterState()
   &SpawnBurst_Instantaneous(Spawn Count=25, Use Spawn Probability=false, UseLoopCountLimit=false)
   &SpawnBurst_Instantaneous(Spawn Count=5, Spawn Group=1, Use Spawn Probability=false, UseLoopCountLimit=false) {module=SpawnBurst_Instantaneous001}
   &SpawnBurst_Instantaneous(Spawn Count=5, Spawn Group=2, Use Spawn Probability=false, UseLoopCountLimit=false) {module=SpawnBurst_Instantaneous002}
   &SpawnBurst_Instantaneous(
    Spawn Count=&RandomRangeInt(Minimum=2, Maximum=8)
    Spawn Group=3
    Use Spawn Probability=false
    UseLoopCountLimit=false
   ) {module=SpawnBurst_Instantaneous003}
   &SpawnRate(SpawnRate=50, Use Spawn Probability=false) {enabled=false}
  ]
  particleSpawnScript [
   &VC_PaintedVariantsOverride() {enabled=false}
   &InitializeParticle(
    Color=&VC_DynamicScaleColor(
     Color to Scale:LinearColor=&MakeCustomLinearColorFromBool(Bool=$User.bIsPainted, True Color=$User.PaintedColor, False Color=<rgba 0 0 0 5>)
     RGB Scale:Vector3f=<1 1 1>
     Alpha Scale:NiagaraFloat=1
    )
    Lifetime=$Emitter.CurrentLoopDuration
    Mesh Scale=&VC_VectorBySpawnGroup(
     SpawnGroup or ExecIndex:NiagaraInt32=0
     CustomInt:NiagaraInt32=0
     MaximumSpawnGroups:NiagaraInt32=4
     Vector Group 0:Vector3f=&Multiply_VectorByFloat(Vector=<0.22 1 1>, Float=0.8)
     Vector Group 1:Vector3f=&Multiply_VectorByFloat(Vector=<0.22 1 1>, Float=0.5)
     Vector Group 2:Vector3f=&Multiply_VectorByFloat(Vector=<0.22 1 1>, Float=0.25)
     Vector Group 3:Vector3f=&Multiply_VectorByFloat(Vector=<0.18 0.18 0.18>, Float=2)
     Vector Group 4:Vector3f=<0 0 0>
     Vector Group 5:Vector3f=<0 0 0>
    )
    Mesh Scale Mode=::NewEnumerator4
    Position Offset=&AddVectorToPosition(Position=<-60 0 0>, Vector=$System.FinalPosition)
    Position Offset Coordinate Space=::Local
    Sprite Rotation Mode=::NewEnumerator2
    UsePositionOffset=true
    Write Material Random=true
   )
   &AddVelocity(Velocity=<-1500 0 0>) {enabled=false}
   &CameraOffset(Camera Offset Amount=-5)
   &InitialMeshOrientation(
    Mesh Axis To Orient=&VC_VectorBySpawnGroup(
     SpawnGroup or ExecIndex:NiagaraInt32=0
     CustomInt:NiagaraInt32=0
     MaximumSpawnGroups:NiagaraInt32=4
     Vector Group 0:Vector3f=<0 0 0>
     Vector Group 1:Vector3f=<0 0 0>
     Vector Group 2:Vector3f=<0 0 0>
     Vector Group 3:Vector3f=<0 1 0>
     Vector Group 4:Vector3f=<0 0 0>
     Vector Group 5:Vector3f=<0 0 0>
    )
    Mesh Orientation Mode=::NewEnumerator2
    Use Rotation Vector=false
   )
  ]
  particleUpdateScript [
   /SetVariables(
    Particles.InnerColor:LinearColor=&VC_DynamicScaleColor(
     Color to Scale:LinearColor=&MakeCustomLinearColorFromBool(Bool=$User.bIsPainted, True Color=$User.InnerCColor, False Color=<rgba 0.8354 0.6922 2 1>)
     RGB Scale:Vector3f=<1 1 1>
     Alpha Scale:NiagaraFloat=1
    )
    Particles.MidColor:LinearColor=&VC_DynamicScaleColor(
     Color to Scale:LinearColor=&MakeCustomLinearColorFromBool(Bool=$User.bIsPainted, True Color=$User.MidCColor, False Color=<rgba 0.1293 0.0781 1 1>)
     RGB Scale:Vector3f=<1 1 1>
     Alpha Scale:NiagaraFloat=1
    )
    Particles.PTagVis:NiagaraInt32=&SpawnGroup()
    Particles.SpiralColor:LinearColor=&VC_DynamicScaleColor(
     Color to Scale:LinearColor=&MakeCustomLinearColorFromBool(Bool=$User.bIsPainted, True Color=$User.SpiralColor, False Color=<rgba 0.4 0.1 2 1>)
     RGB Scale:Vector3f=<1 1 1>
     Alpha Scale:NiagaraFloat=1
    )
   ) {isSetParameters=true}
   &ParticleState()
   &SolveForcesAndVelocity(Clamp Velocity=false, Limit Acceleration=false)
   &ScaleColor(Scale RGB=<5 5 5>, ScaleA=true, ScaleRGB=true)
  ]
  rendererClasses [&NiagaraSpriteRendererProperties, &NiagaraMeshRendererProperties]
  renderers [
   {rendererClass=&NiagaraSpriteRendererProperties}
   {rendererClass=&NiagaraMeshRendererProperties, rendererIndex=1}
   {rendererClass=&NiagaraMeshRendererProperties, rendererIndex=2}
   {rendererClass=&NiagaraMeshRendererProperties, rendererIndex=3}
   {rendererClass=&NiagaraMeshRendererProperties, rendererIndex=4}
  ]
 }
}
summary {
 systemName=NS_Boost_DarkFlame
 emitters [
  {emitterName=NE_VC_SingleBoost_Sprite_Parent, rendererClasses [&NiagaraSpriteRendererProperties]}
  {emitterName=NE_VC_SingleBoost_Ribbon_Parent, rendererClasses [&NiagaraRibbonRendererProperties]}
  {
   emitterName=NE_VC_SingleBoost_Sprite_Parent001
   rendererClasses [&NiagaraSpriteRendererProperties, &NiagaraMeshRendererProperties]
  }
 ]
 userVariables [
  User.bIsFrontendPreview {type=NiagaraBool}
  User.bIsPainted {type=NiagaraBool}
  User.CloudsColor {type=LinearColor, description="", defaultValue {struct=LinearColor, value=<rgba 1 1 1 1>}}
  User.ColorVariantID {type=NiagaraInt32}
  User.DartsColor {type=LinearColor, description="", defaultValue {struct=LinearColor, value=<rgba 1 1 1 1>}}
  User.DTCarOffset {type=Vector3f, description="", defaultValue {struct=Vector3f, value=<0 0 0>}}
  User.GameModeID {type=NiagaraInt32}
  User.GreyTColor {type=LinearColor, description="", defaultValue {struct=LinearColor, value=<rgba 1 1 1 1>}}
  User.InnerCColor {type=LinearColor, description="", defaultValue {struct=LinearColor, value=<rgba 1 1 1 1>}}
  User.MidCColor {type=LinearColor, description="", defaultValue {struct=LinearColor, value=<rgba 1 1 1 1>}}
  User.MTrailColor {type=LinearColor, description="", defaultValue {struct=LinearColor, value=<rgba 1 1 1 1>}}
  User.PaintedColor {type=LinearColor, description="", defaultValue {struct=LinearColor, value=<rgba 1 1 1 1>}}
  User.PaintedScalar {type=NiagaraFloat, defaultValue=1, description=""}
  User.PortCount {type=NiagaraInt32, defaultValue=1}
  User.SingleBoostPos {type=NiagaraPosition, description="", defaultValue {}}
  User.SndTColor {type=LinearColor, description="", defaultValue {struct=LinearColor, value=<rgba 1 1 1 1>}}
  User.SpiralColor {type=LinearColor, description="", defaultValue {struct=LinearColor, value=<rgba 1 1 1 1>}}
 ]
}
