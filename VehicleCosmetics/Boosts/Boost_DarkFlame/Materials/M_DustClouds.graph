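# Translucent unlit Niagara particle material for dust clouds in the Dark Flame vehicle boost cosmetic.
# It fakes lighting by dotting the smoke normal texture's RGB with a fixed vector, then biasing, tinting by particle color and scaling by dynamic Brightness.
# Texture alpha is shaped by dynamic Dissolve and Sharpness, then multiplied by particle alpha.
# Shared functions apply depth fade to opacity and adjustments to emissive.
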Material /VehicleCosmetics/Boosts/Boost_DarkFlame/Materials/M_DustClouds
graph {
 properties {
  bAutomaticallySetUsageInEditor=false
  bUsedWithNiagaraMeshParticles=true
  bUsedWithNiagaraRibbons=true
  bUsedWithNiagaraSprites=true
  blendMode=BLEND_Translucent
  shadingModel=MSM_Unlit
  usedShadingModels=Unlit
 }
 nodes [
  MaterialExpressionVertexColor_1 = ParticleColor()
  n1 = DynamicParameter() {paramNames [Dissolve,Sharpness,Brightness,Param4]}
  n2 = TextureSample() {texture=/DealershipSharedLibrary/FX/Textures/Flipbook/T_VCVFX_SmokeHQ_N}
  MaterialExpressionConstantClamp_1 = Saturate(None=Multiply(A=Power(Base=@n2.A, Exp=@n1.Dissolve), B=@n1.Sharpness))
  MaterialFunctionCall(Opacity In=Multiply(A=@MaterialExpressionVertexColor_1.A, B=@MaterialExpressionConstantClamp_1)) {materialFunction=&MF_VCVFX_DepthFade}
  n3 = ConstantBiasScale(None=DotProduct(A=@n2.RGB, B=Constant3Vector() {constant=<rgba 1 1 0 0>})) {bias=0.2}
  MaterialFunctionCall(Emissive=Multiply(A=Multiply(A=@n3, B=@MaterialExpressionVertexColor_1.RGB), B=@n1.Brightness)) {materialFunction=&MF_FX_EmissiveAdjustments}
 ]
}
