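# Translucent, unlit, two-sided Niagara material for sprites, ribbons and mesh particles, drawing animated fire for the Dark Flame vehicle boost.
# Two samples of a water-noise normal map pan at different speeds and tilings, then offset an inverted-V gradient to form the flame pattern.
# That pattern, tinted by particle color, feeds the shared emissive-adjustment function.
# It is also multiplied by the undistorted gradient, raised to a power and saturated into a sharpened, tapering mask.
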Material /VehicleCosmetics/Boosts/Boost_DarkFlame/Materials/M_Fire2
graph {
 properties {
  bAutomaticallySetUsageInEditor=false
  bUsedWithNiagaraMeshParticles=true
  bUsedWithNiagaraRibbons=true
  bUsedWithNiagaraSprites=true
  blendMode=BLEND_Translucent
  shadingModel=MSM_Unlit
  twoSided=true
  usedShadingModels=Unlit
 }
 nodes [
  n1 = TextureSample(UVs=Panner(Coordinate=TextureCoordinate() {uTiling=-1, vTiling=0.3}) {speedX=-0.6, speedY=-1}) {samplerType=SAMPLERTYPE_Normal, texture=/DealershipSharedLibrary/FX/Textures/Normal/T_VCVFX_Water03Swizzled_Noise}
  n2 = TextureSample(UVs=Panner(Coordinate=TextureCoordinate() {vTiling=0.5}) {speedY=-0.5}) {samplerType=SAMPLERTYPE_Normal, texture=&T_VCVFX_Water03Swizzled_Noise}
  n3 = Multiply(A=ComponentMask(None=Add(A=@n1.RGB, B=@n2.RGB)) {g=true, r=true}, B=Constant() {r=0.25})
  n4 = Multiply(A=OneMinus(None=TextureCoordinate()), B=Constant() {r=0.6})
  n5 = Multiply(A=ComponentMask(None=Add(A=@n3, B=@n4)) {g=true}, B=Constant() {r=1.5})
  n6 = Multiply(A=ComponentMask(None=@n4) {g=true}, B=Constant() {r=2})
  MaterialExpressionConstantClamp_3 = Saturate(None=Power(Base=Multiply(A=@n5, B=@n6), Exp=Constant() {r=3}))
  MaterialExpressionMeshEmitterVertexColor_0 = ParticleColor()
  MaterialFunctionCall(Emissive=Multiply(A=@MaterialExpressionMeshEmitterVertexColor_0.RGB, B=@n5)) {materialFunction=&MF_FX_EmissiveAdjustments}
 ]
}
