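# Translucent, unlit, two-sided Niagara material for the missile effect in the DarkFlame vehicle boost cosmetic, usable on sprite, ribbon and mesh particles.
# Particle color RGB passes through MF_FX_EmissiveAdjustments for the emissive value.
# Particle alpha multiplies a missile texture that scrolls vertically through a Panner.
# The Speed dynamic parameter scales the panner's time, so each particle system can drive the scroll rate.
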Material /VehicleCosmetics/Boosts/Boost_DarkFlame/Materials/M_Misile
graph {
 properties {
  bAutomaticallySetUsageInEditor=false
  bUsedWithNiagaraMeshParticles=true
  bUsedWithNiagaraRibbons=true
  bUsedWithNiagaraSprites=true
  blendMode=BLEND_Translucent
  shadingModel=MSM_Unlit
  twoSided=true
  usedShadingModels=Unlit
 }
 nodes [
  MaterialExpressionMeshEmitterVertexColor_1 = ParticleColor()
  MaterialFunctionCall(Emissive=@MaterialExpressionMeshEmitterVertexColor_1.RGB) {materialFunction=&MF_FX_EmissiveAdjustments}
  n1 = TextureCoordinate()
  MaterialExpressionMeshEmitterDynamicParameter_0 = DynamicParameter() {paramNames [Speed,Param2,Param3,Param4]}
  n2 = Multiply(A=Time(), B=@MaterialExpressionMeshEmitterDynamicParameter_0.Speed)
  Multiply(
   A=@MaterialExpressionMeshEmitterVertexColor_1.A
   B=TextureSample(UVs=Panner(Coordinate=@n1, Time=@n2) {speedY=1.5}) {texture=/VehicleCosmetics/Boosts/Boost_DarkFlame/Textures/Missile_M_T}.RGB
  )
 ]
}
