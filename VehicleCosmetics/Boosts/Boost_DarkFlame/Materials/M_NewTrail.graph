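# Translucent unlit Niagara material for the Dark Flame vehicle boost trail, with scrolling, distorted flame streaks.
# Two summed panning water-normal pairs warp two trail texture samples, while a second trail texture adds detail scaled by Lightning.
# A V-gradient fades the edges, and an opacity texture and particle alpha shape the alpha mask.
# Painted static switches change particle-color tinting and add an inverted-color term driven by the Black dynamic parameter.
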Material /VehicleCosmetics/Boosts/Boost_DarkFlame/Materials/M_NewTrail
graph {
 properties {
  bAutomaticallySetUsageInEditor=false
  bUsedWithNiagaraMeshParticles=true
  bUsedWithNiagaraRibbons=true
  bUsedWithNiagaraSprites=true
  blendMode=BLEND_Translucent
  shadingModel=MSM_Unlit
  usedShadingModels=Unlit
 }
 nodes [
  Constant3Vector() {constant=<rgba 1 0.1868 0.8035 1>}
  MaterialExpressionVertexColor_4 = ParticleColor()
  n1 = Multiply(A=Constant() {r=0.7}, B=TextureCoordinate())
  n2 = Panner(Coordinate=@n1) {speedX=-2, speedY=-0.5}
  n3 = Panner(Coordinate=@n1) {speedX=-0.05, speedY=0.05}
  n4 = ComponentMask(None=Add(
    A=TextureSample(UVs=@n2) {samplerType=SAMPLERTYPE_Normal, texture=/DealershipSharedLibrary/FX/Textures/Normal/T_VCVFX_Water03Swizzled_Noise}.RGB
    B=TextureSample(UVs=@n3) {samplerType=SAMPLERTYPE_Normal, texture=&T_VCVFX_Water03Swizzled_Noise}.RGB
   )) {g=true, r=true}
  n5 = ScalarParameter() {defaultValue=0.1, parameterName=DistortionIntensity}
  n6 = TextureSample(UVs=Add(A=Panner() {speedX=-1.6}, B=Multiply(A=@n4, B=@n5))) {texture=/VehicleCosmetics/Boosts/Boost_DarkFlame/Textures/TrailTexture_M_T}
  n7 = Multiply(A=Constant() {r=4}, B=@n6.R)
  n8 = Constant() {r=1}
  n9 = Add(A=TextureCoordinate(), B=AppendVector(A=Constant() {r=1}, B=Constant()))
  n10 = Multiply(A=Constant() {r=0.7}, B=TextureCoordinate() {uTiling=2, vTiling=2})
  n11 = Panner(Coordinate=@n10) {speedX=-2, speedY=-0.5}
  n12 = Panner(Coordinate=@n10) {speedX=-0.05, speedY=0.05}
  n13 = ComponentMask(None=Add(
    A=TextureSample(UVs=@n11) {samplerType=SAMPLERTYPE_Normal, texture=&T_VCVFX_Water03Swizzled_Noise}.RGB
    B=TextureSample(UVs=@n12) {samplerType=SAMPLERTYPE_Normal, texture=&T_VCVFX_Water03Swizzled_Noise}.RGB
   )) {g=true, r=true}
  n14 = ScalarParameter() {defaultValue=0.1, parameterName=DistortionIntensity}
  n15 = TextureSample(UVs=Add(A=Panner(Coordinate=@n9) {speedX=-1}, B=Multiply(A=@n13, B=@n14))) {texture=&TrailTexture_M_T}
  n16 = Add(A=@n15.G, B=StaticSwitchParameter(True=Multiply(A=@n7, B=@n8), False=@n7) {parameterName=Painteds})
  n17 = ComponentMask(None=TextureCoordinate()) {g=true}
  n18 = Multiply(A=Constant() {r=5}, B=Multiply(A=@n17, B=OneMinus(None=@n17)))
  n19 = Power(Base=@n18, Exp=Constant() {r=3})
  MaterialExpressionConstantClamp_8 = Saturate(None=Multiply(A=@n16, B=Add(A=@n16, B=@n19)))
  n20 = TextureSample() {texture=/VehicleCosmetics/Boosts/Boost_GoldenOutline/Textures/T_VCVFX_Boost_GoldenOutliner_TrailOpacity}
  MaterialExpressionConstantClamp_5 = Saturate(None=Multiply(A=@MaterialExpressionVertexColor_4.A, B=Multiply(A=@MaterialExpressionConstantClamp_8, B=@n20.RGB)))
  MaterialExpressionVertexColor_2 = ParticleColor()
  n21 = Constant() {r=1}
  n22 = OneMinus(None=Add(A=@n15.G, B=@n6.R))
  n23 = Multiply(A=@n18, B=@n22)
  n24 = Add(A=@n23, B=@n23)
  n25 = Constant() {r=1}
  n26 = ParticleColor()
  n27 = TextureSample(UVs=Panner() {speedX=-2}) {texture=/VehicleCosmetics/Boosts/Boost_DarkFlame/Textures/TrailTexture2_M_T}
  n28 = Multiply(A=@n27.R, B=Constant() {r=0.2})
  n29 = StaticSwitchParameter(True=@n28, False=@n27.R) {parameterName=Painted}
  n30 = ScalarParameter() {defaultValue=0.5, parameterName=Lightning}
  n31 = Multiply(A=Add(A=@n28, B=@n27.B), B=@n30)
  n32 = Add(A=@n31, B=Multiply(A=StaticSwitchParameter(True=@n25, False=@n26.RGB) {parameterName=Painteds}, B=Add(A=@n22, B=@n29)))
  n33 = Multiply(A=@n6.R, B=Constant() {r=40})
  MaterialExpressionConstantClamp_6 = Saturate(None=Multiply(A=Subtract(A=@n32, B=@n33), B=Constant() {r=0.25}))
  n34 = Multiply(A=StaticSwitchParameter(True=@n18, False=@n19) {parameterName=Painteds}, B=@MaterialExpressionConstantClamp_6)
  MaterialExpressionConstantClamp_4 = Saturate(None=@n34)
  n35 = Multiply(A=Add(A=@MaterialExpressionConstantClamp_4, B=Multiply(A=@n21, B=@n24)), B=@MaterialExpressionConstantClamp_4)
  n36 = Multiply(A=@n27.B, B=@n30)
  n37 = Add(A=Multiply(A=@MaterialExpressionVertexColor_2.RGB, B=Add(A=@n35, B=@n36)), B=@n31)
  MaterialExpressionConstantClamp_7 = Saturate(None=OneMinus(None=@n37))
  n38 = Multiply(A=@MaterialExpressionConstantClamp_7, B=DynamicParameter() {defaultValue=<rgba 0 1 1 1>, paramNames [Black,Param2,Param3,Param4]}.Black)
  MaterialFunctionCall(Emissive=StaticSwitchParameter(True=Add(A=@n37, B=@n38), False=@n34) {parameterName=Painteds}) {materialFunction=&MF_FX_EmissiveAdjustments}
  Reroute(None=@n27.R)
 ]
}
