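# Translucent unlit spark material for the DarkFlame vehicle boost effect, enabled for Niagara sprite, ribbon and mesh particles.
# A shared ash mask texture is multiplied by particle color, and the result passes through the MF_FX_EmissiveAdjustments function.
# A second multiply combines particle alpha with the same mask's RGB, so the particle system drives both tint and fade.
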Material /VehicleCosmetics/Boosts/Boost_DarkFlame/Materials/M_Sparks
graph {
 properties {
  bAutomaticallySetUsageInEditor=false
  bUsedWithNiagaraMeshParticles=true
  bUsedWithNiagaraRibbons=true
  bUsedWithNiagaraSprites=true
  blendMode=BLEND_Translucent
  shadingModel=MSM_Unlit
  usedShadingModels=Unlit
 }
 nodes [
  n1 = ParticleColor()
  n2 = TextureSample() {texture=/DealershipSharedLibrary/FX/Textures/Masks/T_VCVFX_Ash_A}
  MaterialFunctionCall(Emissive=Multiply(A=@n1.RGB, B=@n2.RGB)) {materialFunction=&MF_FX_EmissiveAdjustments}
  Multiply(A=@n1.A, B=@n2.RGB)
 ]
}
