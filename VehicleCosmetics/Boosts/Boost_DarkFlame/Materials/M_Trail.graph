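# Unlit translucent Niagara material for the trail of the DarkFlame vehicle boost, enabled for ribbons, sprites and mesh particles.
# Two panning water normal samples are summed and scaled down to distort the scrolling UVs of a wispy trail texture.
# That texture, tinted by particle color, feeds an emissive-adjustment function.
# Opacity multiplies particle alpha, the wisp, a GoldenOutline opacity texture and a rounded U-gradient that hard-clips one end.
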Material /VehicleCosmetics/Boosts/Boost_DarkFlame/Materials/M_Trail
graph {
 properties {
  bAutomaticallySetUsageInEditor=false
  bUsedWithNiagaraMeshParticles=true
  bUsedWithNiagaraRibbons=true
  bUsedWithNiagaraSprites=true
  blendMode=BLEND_Translucent
  shadingModel=MSM_Unlit
  usedShadingModels=Unlit
 }
 nodes [
  MaterialExpressionVertexColor_6 = ParticleColor()
  n1 = Panner(Coordinate=TextureCoordinate() {uTiling=0.6}) {speedX=-1, speedY=-0.06}
  n2 = Panner(Coordinate=TextureCoordinate() {uTiling=0.5, vTiling=0.5}) {speedX=-0.8, speedY=0.3}
  n3 = ComponentMask(None=Add(
    A=TextureSample(UVs=@n1) {samplerType=SAMPLERTYPE_Normal, texture=/DealershipSharedLibrary/FX/Textures/Normal/T_VCVFX_Water03Swizzled_Noise}.RGB
    B=TextureSample(UVs=@n2) {samplerType=SAMPLERTYPE_Normal, texture=&T_VCVFX_Water03Swizzled_Noise}.RGB
   )) {g=true, r=true}
  n4 = Constant() {r=0.05}
  n5 = TextureSample(UVs=Add(A=Multiply(A=@n3, B=@n4), B=Panner() {speedX=-1})) {texture=/VehicleCosmetics/Boosts/Boost_DarkFlame/Textures/WhispyTrail_M_T}
  MaterialFunctionCall(Emissive=Multiply(A=@MaterialExpressionVertexColor_6.RGB, B=@n5.RGB)) {materialFunction=&MF_FX_EmissiveAdjustments}
  n6 = OneMinus(None=ComponentMask(None=TextureCoordinate()) {r=true})
  n7 = TextureSample() {texture=/VehicleCosmetics/Boosts/Boost_GoldenOutline/Textures/T_VCVFX_Boost_GoldenOutliner_TrailOpacity}
  Multiply(A=Multiply(A=@MaterialExpressionVertexColor_6.A, B=Multiply(A=@n5.RGB, B=@n7.RGB)), B=Round(None=Power(Base=@n6) {constExponent=0.2}))
 ]
}
